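# Niagara effect for the SplishSplash backpack cosmetic, producing drifting bubbles and bubble pops at sockets read through a Socket Reader.
# Bubbles spawns curl-noise sprites that briefly inherit source movement.
# Eight system pop booleans, apparently derived from socket-scale comparisons, trigger socket-attached burst flashes and gravity-driven droplets, using spawn groups to select each socket.
# Brightness or opacity varies by time of day, with a front-end override for bubbles.
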NiagaraSystem /BRCosmetics/Accessories/FORT_Backpacks/Backpack_M_MED_SplishSplash/FX/NS_Backpack_SplishSplash
dependencies {
 usedDataInterfaces [&NiagaraDataInterfaceCurve, &NiagaraDataInterfaceSocketReader]
 usedDynamicInputs [&SetBoolByFloatComparison]
 usedModules [
  &self:SystemSpawnScript.SystemScriptSource.SystemScriptGraph.NiagaraNodeAssignment_5.SetVariables_#1
  &SystemState
  &self:SystemSpawnScript.SystemScriptSource.SystemScriptGraph.NiagaraNodeAssignment_0.SetVariables_#2
  &self:GetSocketScale_01
  &self:SystemSpawnScript.SystemScriptSource.SystemScriptGraph.NiagaraNodeAssignment_4.SetVariables_#3
  &self:SystemSpawnScript.SystemScriptSource.SystemScriptGraph.NiagaraNodeAssignment_2.SetVariables_#4
  &self:SystemSpawnScript.SystemScriptSource.SystemScriptGraph.NiagaraNodeAssignment_3.SetVariables_#5
 ]
}
emitters {
 Bubbles {
  emitterUpdateScript [
   &EmitterState(
    Inactive Response=::NewEnumerator1
    Life Cycle Mode=::NewEnumerator1
    Loop Behavior=::NewEnumerator1
    Loop Duration Mode=::NewEnumerator1
    UseLoopDelay=false
   )
   &SpawnRate(Spawn Probability=0.6, SpawnRate=1.4, Use Spawn Probability=true)
  ]
  particleSpawnScript [
   &FN_SetSpawnGroup(Number of Spawn Groups=2)
   &InitializeParticle(
    Color Maximum=<rgba 0.1157 0.7315 1 1>
    Color Minimum=<rgba 0.0934 0.4532 0.6406 1>
    Color Mode=::NewEnumerator2
    Lifetime=&RandomRangeFloat(Minimum=1.4, Maximum=1.4)
    Mesh Scale Mode=::NewEnumerator3
    Mesh Uniform Scale=0.3
    Position Mode=::NewEnumerator0
    Sprite Size Mode=::NewEnumerator3
    Uniform Sprite Size=&RandomRangeFloat(Minimum=2, Maximum=3.5)
    Write Material Random=true
   )
   &DynamicMaterialParameters(
    Index 0 Param 1=1
    Module.Index0Param0Write=true
    Module.Index0Param1Write=false
    Module.Index0Param2Write=false
    Module.Index0Param3Write=false
   )
   &SocketLocation(
    SamplePosOffset=true
    Sampled Position Offset=&RandomRangeVector(Minimum=<-2 -2 0>, Maximum=<2 2 0>)
    Socket Index=&SpawnGroup()
    Socket Reader=$System.Socket Reader
    Socket Sampling Mode=::NewEnumerator3
    VelocitySpeedLimit=false
   )
   &AddVelocity(Velocity=&RandomRangeVector(Minimum=<-1 -2 1>, Maximum=<1 0 4>), Velocity Speed Scale=3)
  ]
  particleUpdateScript [
   &ParticleState()
   &InheritSourceMovement(
    Applied Position Scale=&VectorFromFloat(Value=&FloatFromCurve(FloatCurve=DataInterface() {propertyValues {Curve {keys [<at 0.5054 1 RCIM_Cubic>,<at 0.6594 0>]}}}))
    ApplyForce=false
    ApplyPosition=true
    ApplyPositionFalloff=false
   )
   &CurlNoiseForce(
    Noise Strength=&FloatFromCurve(
     FloatCurve=DataInterface() {propertyValues {Curve {keys [<at 0 0.05 RCIM_Cubic>, {arriveTangent=0.0079, interpMode=RCIM_Cubic, tangentMode=RCTM_Break, time=1, value=1}]}}}
     Scale Curve=10
    )
    Pan Noise=false
   )
   &ScaleColor(
    Scale Alpha=&FloatFromCurve(
     FloatCurve=DataInterface() {propertyValues {Curve {keys [<at 0 0.6 RCIM_Cubic>, {arriveTangent=1.2873, interpMode=RCIM_Cubic, tangentMode=RCTM_Break, time=0.9, value=1}, <at 1 0 RCIM_Cubic>]}}}
     Scale Curve=&FN_FrontEndFloatModifier(
      Gameplay=&FN_Select_Float_from_Time_Of_Day(Morning=0.6, Day=0.8, Evening=0.6, Night=0.4)
      Front End=0.75
     )
    )
    Scale RGB=&Multiply_VectorByFloat(Vector=<1 1 1>, Float=2.6)
    ScaleA=true
    ScaleRGB=true
   )
   &ScaleSpriteSize(Scale Factor=&Vector2DFromFloat(Value=&FloatFromCurve(
      FloatCurve=DataInterface() {propertyValues {Curve {keys [
          <at 0 0.4 RCIM_Cubic>
          {arriveTangent=-0.01, interpMode=RCIM_Cubic, leaveTangent=-0.1507, tangentMode=RCTM_Break, time=0.3, value=1}
          {arriveTangent=0.2157, interpMode=RCIM_Cubic, leaveTangent=0.2157, tangentMode=RCTM_User, time=0.9267, value=0.844}
          {arriveTangent=28.6014, interpMode=RCIM_Cubic, tangentMode=RCTM_Break, time=1, value=2.4574}
         ]}}}
      Scale Curve=1.2
     )))
   &SolveForcesAndVelocity(Clamp Velocity=false, Limit Acceleration=false)
  ]
  rendererClasses [&NiagaraSpriteRendererProperties]
  renderers [{rendererClass=&NiagaraSpriteRendererProperties}]
 }
 BurstStaticBubbles {
  emitterUpdateScript [
   &EmitterState(Inactive Response=::NewEnumerator1, Life Cycle Mode=::NewEnumerator1, Loop Duration=1, UseLoopDelay=false)
   &SpawnPerFrame(Spawn:NiagaraBool=$System.BubblePopBool1, Use Spawn Probability=false)
   &SpawnPerFrame(Spawn:NiagaraBool=$System.BubblePopBool2, Spawn Group=1, Use Spawn Probability=false) {module=SpawnPerFrame007}
   &SpawnPerFrame(Spawn:NiagaraBool=$System.BubblePopBool3, Spawn Group=2, Use Spawn Probability=false) {module=SpawnPerFrame006}
   &SpawnPerFrame(Spawn:NiagaraBool=$System.BubblePopBool4, Spawn Group=3, Use Spawn Probability=false) {module=SpawnPerFrame005}
   &SpawnPerFrame(Spawn:NiagaraBool=$System.BubblePopBool5, Spawn Group=4, Use Spawn Probability=false) {module=SpawnPerFrame004}
   &SpawnPerFrame(Spawn:NiagaraBool=$System.BubblePopBool6, Spawn Group=5, Use Spawn Probability=false) {module=SpawnPerFrame003}
   &SpawnPerFrame(Spawn:NiagaraBool=$System.BubblePopBool7, Spawn Group=6, Use Spawn Probability=false) {module=SpawnPerFrame002}
   &SpawnPerFrame(Spawn:NiagaraBool=$System.BubblePopBool8, Spawn Group=7, Use Spawn Probability=false) {module=SpawnPerFrame001}
  ]
  particleSpawnScript [
   &InitializeParticle(
    Lifetime=0.14
    Mesh Scale Mode=::NewEnumerator3
    Sprite Rotation Mode=::NewEnumerator2
    Sprite Size Mode=::NewEnumerator3
    Uniform Sprite Size=&RandomRangeFloat(Minimum=13, Maximum=16)
    UsePositionOffset=false
    Write Material Random=true
   )
   &DynamicMaterialParameters(
    Index 0 Param 1=5
    Index 0 Param 2=1
    Index 0 Param 3=1
    Index 0 Param 4=1
    Index 1 Param 1=6.667
    Index 1 Param 3=1
    Index 1 Param 4=1
    Module.Index0Float0Write=true
    Module.Index0Float1Write=true
    Module.Index0Float2Write=true
    Module.Index0Float3Write=true
    Module.Index1Float0Write=true
    Module.Index1Float1Write=true
    Module.Index1Float2Write=true
    Module.Index1Float3Write=true
    Write Parameter Index 1=::NewEnumerator1
   )
  ]
  particleUpdateScript [
   &ParticleState()
   &ScaleSpriteSize(
    Scale Sprite Size Mode=::NewEnumerator0
    Uniform Scale Factor=&FloatFromCurve(FloatCurve=DataInterface() {propertyValues {Curve {keys [
         {arriveTangent=1.3698, interpMode=RCIM_Cubic, leaveTangent=1.3698, tangentMode=RCTM_User, time=0, value=0.4}
         {tangentMode=RCTM_User, time=1, value=1}
        ]}}})
   )
   &SocketLocation(
    SamplePosOffset=true
    Socket Index=&MergeHelper_IntSwitch(0=2, 1=3, 2=4, 3=5, 4=6, 5=7, 6=8, 7=9, Index=&SpawnGroup(), Num=::NewEnumerator6)
    Socket Reader=$System.Socket Reader
    Socket Sampling Mode=::NewEnumerator3
    VelocitySpeedLimit=false
   )
   &ScaleColor(
    Scale Alpha=&FloatFromCurve(FloatCurve=DataInterface() {propertyValues {Curve {keys [<at 0.2917 1 RCIM_Cubic>,<at 1 0>]}}}, Scale Curve=0.4)
    Scale RGB=&VectorFromFloat(Value=&FloatFromCurve(
      FloatCurve=DataInterface() {propertyValues {Curve {keys [
          <at 0 3 RCIM_Cubic>
          {arriveTangent=-0.9668, interpMode=RCIM_Cubic, leaveTangent=-0.9668, tangentMode=RCTM_User, time=0.3086, value=1.5}
          <at 1 1 RCIM_Cubic>
         ]}}}
      Scale Curve=&FNTimeOfDayScalar(Night Scale=0.6, Morning Scale=0.9, Evening Scale=0.8)
     ))
    ScaleA=true
    ScaleRGB=true
   )
  ]
  rendererClasses [&NiagaraSpriteRendererProperties]
  renderers [{rendererClass=&NiagaraSpriteRendererProperties}]
 }
 PopDrops {
  emitterUpdateScript [
   &EmitterState(Inactive Response=::NewEnumerator1, Life Cycle Mode=::NewEnumerator1, Loop Duration=1, UseLoopDelay=false)
   /SetVariables(Emitter.RandomSpawn:NiagaraInt32=&RandomRangeInt(Minimum=2, Maximum=3)) {isSetParameters=true}
   &SpawnPerFrame(Spawn:NiagaraBool=$System.BubblePopBool1, Spawn Count:NiagaraInt32=$Emitter.RandomSpawn, Use Spawn Probability=false)
   &SpawnPerFrame(Spawn:NiagaraBool=$System.BubblePopBool2, Spawn Count:NiagaraInt32=$Emitter.RandomSpawn, Spawn Group=1, Use Spawn Probability=false) {module=SpawnPerFrame007}
   &SpawnPerFrame(Spawn:NiagaraBool=$System.BubblePopBool3, Spawn Count:NiagaraInt32=$Emitter.RandomSpawn, Spawn Group=2, Use Spawn Probability=false) {module=SpawnPerFrame006}
   &SpawnPerFrame(Spawn:NiagaraBool=$System.BubblePopBool4, Spawn Count:NiagaraInt32=$Emitter.RandomSpawn, Spawn Group=3, Use Spawn Probability=false) {module=SpawnPerFrame005}
   &SpawnPerFrame(Spawn:NiagaraBool=$System.BubblePopBool5, Spawn Count:NiagaraInt32=$Emitter.RandomSpawn, Spawn Group=4, Use Spawn Probability=false) {module=SpawnPerFrame004}
   &SpawnPerFrame(Spawn:NiagaraBool=$System.BubblePopBool6, Spawn Count:NiagaraInt32=$Emitter.RandomSpawn, Spawn Group=5, Use Spawn Probability=false) {module=SpawnPerFrame003}
   &SpawnPerFrame(Spawn:NiagaraBool=$System.BubblePopBool7, Spawn Count:NiagaraInt32=$Emitter.RandomSpawn, Spawn Group=6, Use Spawn Probability=false) {module=SpawnPerFrame002}
   &SpawnPerFrame(Spawn:NiagaraBool=$System.BubblePopBool8, Spawn Count:NiagaraInt32=$Emitter.RandomSpawn, Spawn Group=7, Use Spawn Probability=false) {module=SpawnPerFrame001}
  ]
  particleSpawnScript [
   &InitializeParticle(
    Lifetime=&RandomRangeFloat(Minimum=0.3, Maximum=0.6)
    Mesh Scale Mode=::NewEnumerator3
    Sprite Rotation Mode=::NewEnumerator2
    Sprite Size=&RandomRangeVector2D(Minimum=<0.5 2>, Maximum=<1 3>)
    Sprite Size Mode=::NewEnumerator4
    UsePositionOffset=false
    Write Material Random=true
   )
   &SocketLocation(
    SamplePosOffset=true
    Sampled Position Offset=&RandomRangeVector(Minimum=<-1 -1 -1>, Maximum=<1 1 1>)
    Socket Index=&MergeHelper_IntSwitch(0=2, 1=3, 2=4, 3=5, 4=6, 5=7, 6=8, 7=9, Index=&SpawnGroup(), Num=::NewEnumerator6)
    Socket Reader=$System.Socket Reader
    Socket Sampling Mode=::NewEnumerator3
    VelocitySpeedLimit=false
   )
   &AddVelocity(Velocity=&RandomRangeVector(Minimum=<-15 -15 5>, Maximum=<15 15 21>), Velocity Speed Scale=1.5)
  ]
  particleUpdateScript [
   &ParticleState()
   &ScaleSpriteSize(
    Scale Sprite Size Mode=::NewEnumerator0
    Uniform Scale Factor=&FloatFromCurve(FloatCurve=DataInterface() {propertyValues {Curve {keys [
         {arriveTangent=1.3698, interpMode=RCIM_Cubic, leaveTangent=1.3698, tangentMode=RCTM_User, time=0, value=0.3}
         {tangentMode=RCTM_User, time=1, value=1}
        ]}}})
   )
   &DynamicMaterialParameters(
    Index 0 Param 3=1
    Index 0 Param 4=1
    Module.Param0WriteEnabled=true
    Module.Param1WriteEnabled=true
    Module.Param2WriteEnabled=true
    Module.Param3WriteEnabled=true
   ) {module=DynamicMaterialParameters001}
   &GravityForce(Gravity=&RandomRangeVector(Minimum=<0 0 -120>, Maximum=<0 0 -70>))
   &SolveForcesAndVelocity(Clamp Velocity=false, Limit Acceleration=false)
   &ScaleColor(
    Scale Alpha=&FloatFromCurve(
     FloatCurve=DataInterface() {propertyValues {Curve {keys [<at 0.2917 1 RCIM_Cubic>,<at 1 0>]}}}
     Scale Curve=&FNTimeOfDayScalar(Night Scale=0.75, Morning Scale=0.9, Evening Scale=0.8)
    )
    ScaleA=true
    ScaleRGB=true
   )
  ]
  rendererClasses [&NiagaraSpriteRendererProperties]
  renderers [{rendererClass=&NiagaraSpriteRendererProperties}]
 }
}
summary {
 systemName=NS_Backpack_SplishSplash
 emitters [
  {emitterName=Bubbles, rendererClasses [&NiagaraSpriteRendererProperties]}
  {emitterName=BurstStaticBubbles, rendererClasses [&NiagaraSpriteRendererProperties]}
  {emitterName=PopDrops, rendererClasses [&NiagaraSpriteRendererProperties]}
 ]
 userVariables [
  User.bIsFrontEnd {type=NiagaraBool}
  User.bIsFrontEndPreview {type=NiagaraBool}
 ]
}
