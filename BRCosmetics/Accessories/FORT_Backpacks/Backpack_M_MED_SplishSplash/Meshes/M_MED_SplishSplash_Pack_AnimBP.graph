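# Animation Blueprint for the SplishSplash backpack cosmetic, adding secondary motion to its towel, llama head and bubbles.
# AnimGraph chains additive Modify Bone nodes on base and dynamic bubble bones, then runs spring-constrained AnimDynamics on towel, head and base bubbles.
# EventGraph computes time-based sine wobble rotations and Atan(Tan) sawtooth bubble scales.
# It also updates gravity override, simulation space and zipline/front-end alpha variables from character state.
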AnimBlueprint /BRCosmetics/Accessories/FORT_Backpacks/Backpack_M_MED_SplishSplash/Meshes/M_MED_SplishSplash_Pack_AnimBP
target_skeleton=&/M_MED_SplishSplash_Pack_Skeleton
graphs [
 AnimGraph {nodes [
   Output Pose(Result=Use cached pose 'CR_Cache_dynamics'()) {layerGroup=None}
   n1 = "Transform (Modify) Bone - Bone: base_bubble_01"(ComponentPose=Local To Component(LocalPose=Input Pose()), Alpha=1, Translation="0,0,0", Rotation="0,0,0", Scale="1,1,1") {
    rotationMode=BMM_Additive
    scaleMode=BMM_Additive
    translationMode=BMM_Additive
    alphaBoolBlend {blendOption=Linear}
    boneToModify {boneName=base_bubble_01}
   }
   n2 = "Transform (Modify) Bone - Bone: base_bubble_02"(ComponentPose=@n1, Alpha=1, Translation="0,0,0", Rotation="0,0,0", Scale="1,1,1") {
    rotationMode=BMM_Additive
    scaleMode=BMM_Additive
    translationMode=BMM_Additive
    alphaBoolBlend {blendOption=Linear}
    boneToModify {boneName=base_bubble_02}
   }
   n3 = "Transform (Modify) Bone - Bone: base_bubble_03"(ComponentPose=@n2, Alpha=1, Translation="0,0,0", Rotation="0,0,0", Scale="1,1,1") {
    rotationMode=BMM_Additive
    scaleMode=BMM_Additive
    translationMode=BMM_Additive
    alphaBoolBlend {blendOption=Linear}
    boneToModify {boneName=base_bubble_03}
   }
   n4 = "Transform (Modify) Bone - Bone: base_bubble_04"(ComponentPose=@n3, Alpha=1, Translation="0,0,0", Rotation="0,0,0", Scale="1,1,1") {
    rotationMode=BMM_Additive
    scaleMode=BMM_Additive
    translationMode=BMM_Additive
    alphaBoolBlend {blendOption=Linear}
    boneToModify {boneName=base_bubble_04}
   }
   n5 = "Transform (Modify) Bone - Bone: base_bubble_05"(ComponentPose=@n4, Alpha=1, Translation="0,0,0", Rotation="0,0,0", Scale="1,1,1") {
    rotationMode=BMM_Additive
    scaleMode=BMM_Additive
    translationMode=BMM_Additive
    alphaBoolBlend {blendOption=Linear}
    boneToModify {boneName=base_bubble_05}
   }
   n6 = "Transform (Modify) Bone - Bone: base_bubble_06"(ComponentPose=@n5, Alpha=1, Translation="0,0,0", Rotation="0,0,0", Scale="1,1,1") {
    rotationMode=BMM_Additive
    scaleMode=BMM_Additive
    translationMode=BMM_Additive
    alphaBoolBlend {blendOption=Linear}
    boneToModify {boneName=base_bubble_06}
   }
   n7 = "Transform (Modify) Bone - Bone: base_bubble_07"(ComponentPose=@n6, Alpha=1, Translation="0,0,0", Rotation="0,0,0", Scale="1,1,1") {
    rotationMode=BMM_Additive
    scaleMode=BMM_Additive
    translationMode=BMM_Additive
    alphaBoolBlend {blendOption=Linear}
    boneToModify {boneName=base_bubble_07}
   }
   n8 = "Transform (Modify) Bone - Bone: base_bubble_08"(ComponentPose=@n7, Alpha=1, Translation="0,0,0", Rotation="0,0,0", Scale="1,1,1") {
    rotationMode=BMM_Additive
    scaleMode=BMM_Additive
    translationMode=BMM_Additive
    alphaBoolBlend {blendOption=Linear}
    boneToModify {boneName=base_bubble_08}
   }
   n9 = "Transform (Modify) Bone - Bone: dyn_bubble_01"(ComponentPose=@n8, Alpha=1, Translation="0,0,0", Rotation="0,0,0", Scale="1,1,1") {
    rotationMode=BMM_Additive
    rotationSpace=BCS_BoneSpace
    scaleMode=BMM_Additive
    translationMode=BMM_Additive
    alphaBoolBlend {blendOption=Linear}
    boneToModify {boneName=dyn_bubble_01}
   }
   n10 = "Transform (Modify) Bone - Bone: dyn_bubble_02"(ComponentPose=@n9, Alpha=1, Translation="0,0,0", Rotation="0,0,0", Scale="1,1,1") {
    rotationMode=BMM_Additive
    rotationSpace=BCS_BoneSpace
    scaleMode=BMM_Additive
    translationMode=BMM_Additive
    alphaBoolBlend {blendOption=Linear}
    boneToModify {boneName=dyn_bubble_02}
   }
   n11 = "Transform (Modify) Bone - Bone: dyn_bubble_03"(ComponentPose=@n10, Alpha=1, Translation="0,0,0", Rotation="0,0,0", Scale="1,1,1") {
    rotationMode=BMM_Additive
    rotationSpace=BCS_BoneSpace
    scaleMode=BMM_Additive
    translationMode=BMM_Additive
    alphaBoolBlend {blendOption=Linear}
    boneToModify {boneName=dyn_bubble_03}
   }
   n12 = "Transform (Modify) Bone - Bone: dyn_bubble_04"(ComponentPose=@n11, Alpha=1, Translation="0,0,0", Rotation="0,0,0", Scale="1,1,1") {
    rotationMode=BMM_Additive
    rotationSpace=BCS_BoneSpace
    scaleMode=BMM_Additive
    translationMode=BMM_Additive
    alphaBoolBlend {blendOption=Linear}
    boneToModify {boneName=dyn_bubble_04}
   }
   n13 = "Transform (Modify) Bone - Bone: dyn_bubble_05"(ComponentPose=@n12, Alpha=1, Translation="0,0,0", Rotation="0,0,0", Scale="1,1,1") {
    rotationMode=BMM_Additive
    rotationSpace=BCS_BoneSpace
    scaleMode=BMM_Additive
    translationMode=BMM_Additive
    alphaBoolBlend {blendOption=Linear}
    boneToModify {boneName=dyn_bubble_05}
   }
   n14 = "Transform (Modify) Bone - Bone: dyn_bubble_06"(ComponentPose=@n13, Alpha=1, Translation="0,0,0", Rotation="0,0,0", Scale="1,1,1") {
    rotationMode=BMM_Additive
    rotationSpace=BCS_BoneSpace
    scaleMode=BMM_Additive
    translationMode=BMM_Additive
    alphaBoolBlend {blendOption=Linear}
    boneToModify {boneName=dyn_bubble_06}
   }
   n15 = "Transform (Modify) Bone - Bone: dyn_bubble_07"(ComponentPose=@n14, Alpha=1, Translation="0,0,0", Rotation="0,0,0", Scale="1,1,1") {
    rotationMode=BMM_Additive
    rotationSpace=BCS_BoneSpace
    scaleMode=BMM_Additive
    translationMode=BMM_Additive
    alphaBoolBlend {blendOption=Linear}
    boneToModify {boneName=dyn_bubble_07}
   }
   n16 = "Transform (Modify) Bone - Bone: dyn_bubble_08"(ComponentPose=@n15, Alpha=1, Translation="0,0,0", Rotation="0,0,0", Scale="1,1,1") {
    rotationMode=BMM_Additive
    rotationSpace=BCS_BoneSpace
    scaleMode=BMM_Additive
    translationMode=BMM_Additive
    alphaBoolBlend {blendOption=Linear}
    boneToModify {boneName=dyn_bubble_08}
   }
   Save cached pose 'CR_Cache_Input'(Pose=Component To Local(ComponentPose=@n16)) {cacheName=CR_Cache_Input}
   n17 = Anim Dynamics - Bone: dyn_towel_01!(
    SimulationSpace:AnimPhysSimSpaceType Enum
    ComponentPose=Local To Component(LocalPose=Use cached pose 'CR_Cache_Input'())
    Alpha=1
    GravityOverride:Vector="0,0,-980"
    ExternalForce="0,0,0"
   ) -> (Pose) {
    angularDampingOverride=1
    angularSpringConstant=800
    bAngularSpring=true
    bOverrideAngularDamping=true
    bUseGravityOverride=true
    gravityScale=0.35
    simulationSpace=RootRelative
    alphaBoolBlend {blendOption=Linear}
    boundBone {boneName=dyn_towel_01}
    componentAppliedLinearAccClamp=<10000 10000 10000>
    componentLinearAccScale=<1 1 1>
    componentLinearVelScale=<1 1 1>
    gravityOverride {z=-980}
    physicsBodyDefinitions [{
      boundBone {boneName=dyn_towel_01}
      boxExtents=<30 40 40>
      constraintSetup {angularLimitsMax=<5 20 0>, angularLimitsMin=<-5 -20 -50>, angularTarget=<1 0 0>}
      localJointOffset=<4 0 0>
     }]
   }
   n18 = Anim Dynamics - Bone: dyn_towel_02!(SimulationSpace:AnimPhysSimSpaceType Enum, ComponentPose=@n17, Alpha=1, GravityOverride:Vector="0,0,-980", ExternalForce="0,0,0") -> (Pose) {
    angularDampingOverride=1
    angularSpringConstant=800
    bAngularSpring=true
    bOverrideAngularDamping=true
    bUseGravityOverride=true
    gravityScale=0.35
    simulationSpace=RootRelative
    alphaBoolBlend {blendOption=Linear}
    boundBone {boneName=dyn_towel_02}
    componentAppliedLinearAccClamp=<10000 10000 10000>
    componentLinearAccScale=<1 1 1>
    componentLinearVelScale=<1 1 1>
    gravityOverride {z=-980}
    physicsBodyDefinitions [{
      boundBone {boneName=dyn_towel_02}
      boxExtents=<30 40 40>
      constraintSetup {angularLimitsMax=<5 15 0>, angularLimitsMin=<-5 -15 -50>, angularTarget=<1 0 0>}
      localJointOffset=<4 0 0>
     }]
   }
   n19 = Anim Dynamics - Bone: dyn_llama_head!(SimulationSpace:AnimPhysSimSpaceType Enum, ComponentPose=@n18, Alpha=1, GravityOverride:Vector="0,0,-980", ExternalForce="0,0,0") -> (Pose) {
    angularDampingOverride=1
    angularSpringConstant=20000
    bAngularSpring=true
    bOverrideAngularDamping=true
    bUseGravityOverride=true
    gravityScale=0.35
    simulationSpace=RootRelative
    alphaBoolBlend {blendOption=Linear}
    boundBone {boneName=dyn_llama_head}
    componentAppliedLinearAccClamp=<10000 10000 10000>
    componentLinearAccScale=<1 1 1>
    componentLinearVelScale=<1 1 1>
    gravityOverride {z=-980}
    physicsBodyDefinitions [{
      boundBone {boneName=dyn_llama_head}
      boxExtents=<30 50 50>
      constraintSetup {angularLimitsMax=<15 25 15>, angularLimitsMin=<-15 -30 -15>, angularTarget=<1 0 0>}
      localJointOffset=<6 0 0>
     }]
   }
   n20 = Anim Dynamics - Bone: base_bubble_01!(SimulationSpace:AnimPhysSimSpaceType Enum, ComponentPose=@n19, Alpha=1, GravityOverride:Vector="0,0,-980", ExternalForce="0,0,0") -> (Pose) {
    angularDampingOverride=1
    angularSpringConstant=20000
    bAngularSpring=true
    bOverrideAngularDamping=true
    bUseGravityOverride=true
    gravityScale=0.35
    simulationSpace=RootRelative
    alphaBoolBlend {blendOption=Linear}
    boundBone {boneName=base_bubble_01}
    componentAppliedLinearAccClamp=<10000 10000 10000>
    componentLinearAccScale=<1 1 1>
    componentLinearVelScale=<1 1 1>
    gravityOverride {z=-980}
    physicsBodyDefinitions [{
      boundBone {boneName=base_bubble_01}
      boxExtents=<30 50 50>
      constraintSetup {angularLimitsMax=<2 3 4>, angularLimitsMin=<-2 -2.5 -2>, angularTarget=<1 0 0>}
      localJointOffset=<4 0 0>
     }]
   }
   n21 = Anim Dynamics - Bone: base_bubble_02!(SimulationSpace:AnimPhysSimSpaceType Enum, ComponentPose=@n20, Alpha=1, GravityOverride:Vector="0,0,-980", ExternalForce="0,0,0") -> (Pose) {
    angularDampingOverride=1
    angularSpringConstant=20000
    bAngularSpring=true
    bOverrideAngularDamping=true
    bUseGravityOverride=true
    gravityScale=0.35
    simulationSpace=RootRelative
    alphaBoolBlend {blendOption=Linear}
    boundBone {boneName=base_bubble_02}
    componentAppliedLinearAccClamp=<10000 10000 10000>
    componentLinearAccScale=<1 1 1>
    componentLinearVelScale=<1 1 1>
    gravityOverride {z=-980}
    physicsBodyDefinitions [{
      boundBone {boneName=base_bubble_02}
      boxExtents=<30 50 50>
      constraintSetup {angularLimitsMax=<2.5 2.5 3>, angularLimitsMin=<-4 -4 -4>, angularTarget=<1 0 0>}
      localJointOffset=<4 0 0>
     }]
   }
   n22 = Anim Dynamics - Bone: base_bubble_03!(SimulationSpace:AnimPhysSimSpaceType Enum, ComponentPose=@n21, Alpha=1, GravityOverride:Vector="0,0,-980", ExternalForce="0,0,0") -> (Pose) {
    angularDampingOverride=1
    angularSpringConstant=20000
    bAngularSpring=true
    bOverrideAngularDamping=true
    bUseGravityOverride=true
    gravityScale=0.35
    simulationSpace=RootRelative
    alphaBoolBlend {blendOption=Linear}
    boundBone {boneName=base_bubble_03}
    componentAppliedLinearAccClamp=<10000 10000 10000>
    componentLinearAccScale=<1 1 1>
    componentLinearVelScale=<1 1 1>
    gravityOverride {z=-980}
    physicsBodyDefinitions [{
      boundBone {boneName=base_bubble_03}
      boxExtents=<30 50 50>
      constraintSetup {angularLimitsMax=<2.5 3 4>, angularLimitsMin=<-2 -1.5 -3>, angularTarget=<1 0 0>}
      localJointOffset=<4 0 0>
     }]
   }
   n23 = Anim Dynamics - Bone: base_bubble_04!(SimulationSpace:AnimPhysSimSpaceType Enum, ComponentPose=@n22, Alpha=1, GravityOverride:Vector="0,0,-980", ExternalForce="0,0,0") -> (Pose) {
    angularDampingOverride=1
    angularSpringConstant=20000
    bAngularSpring=true
    bOverrideAngularDamping=true
    bUseGravityOverride=true
    gravityScale=0.35
    simulationSpace=RootRelative
    alphaBoolBlend {blendOption=Linear}
    boundBone {boneName=base_bubble_04}
    componentAppliedLinearAccClamp=<10000 10000 10000>
    componentLinearAccScale=<1 1 1>
    componentLinearVelScale=<1 1 1>
    gravityOverride {z=-980}
    physicsBodyDefinitions [{
      boundBone {boneName=base_bubble_04}
      boxExtents=<30 50 50>
      constraintSetup {angularLimitsMax=<3 3 3>, angularLimitsMin=<-4 -4 -4>, angularTarget=<1 0 0>}
      localJointOffset=<4 0 0>
     }]
   }
   n24 = Anim Dynamics - Bone: base_bubble_05!(SimulationSpace:AnimPhysSimSpaceType Enum, ComponentPose=@n23, Alpha=1, GravityOverride:Vector="0,0,-980", ExternalForce="0,0,0") -> (Pose) {
    angularDampingOverride=1
    angularSpringConstant=20000
    bAngularSpring=true
    bOverrideAngularDamping=true
    bUseGravityOverride=true
    gravityScale=0.35
    simulationSpace=RootRelative
    alphaBoolBlend {blendOption=Linear}
    boundBone {boneName=base_bubble_05}
    componentAppliedLinearAccClamp=<10000 10000 10000>
    componentLinearAccScale=<1 1 1>
    componentLinearVelScale=<1 1 1>
    gravityOverride {z=-980}
    physicsBodyDefinitions [{
      boundBone {boneName=base_bubble_05}
      boxExtents=<30 50 50>
      constraintSetup {angularLimitsMax=<3 2 4>, angularLimitsMin=<-2.5 -3.5 -3.5>, angularTarget=<1 0 0>}
      localJointOffset=<4 0 0>
     }]
   }
   n25 = Anim Dynamics - Bone: base_bubble_06!(SimulationSpace:AnimPhysSimSpaceType Enum, ComponentPose=@n24, Alpha=1, GravityOverride:Vector="0,0,-980", ExternalForce="0,0,0") -> (Pose) {
    angularDampingOverride=1
    angularSpringConstant=20000
    bAngularSpring=true
    bOverrideAngularDamping=true
    bUseGravityOverride=true
    gravityScale=0.35
    simulationSpace=RootRelative
    alphaBoolBlend {blendOption=Linear}
    boundBone {boneName=base_bubble_06}
    componentAppliedLinearAccClamp=<10000 10000 10000>
    componentLinearAccScale=<1 1 1>
    componentLinearVelScale=<1 1 1>
    gravityOverride {z=-980}
    physicsBodyDefinitions [{
      boundBone {boneName=base_bubble_06}
      boxExtents=<30 50 50>
      constraintSetup {angularLimitsMax=<2 3 4>, angularLimitsMin=<-3.5 -3.5 -3.5>, angularTarget=<1 0 0>}
      localJointOffset=<4 0 0>
     }]
   }
   n26 = Anim Dynamics - Bone: base_bubble_07!(SimulationSpace:AnimPhysSimSpaceType Enum, ComponentPose=@n25, Alpha=1, GravityOverride:Vector="0,0,-980", ExternalForce="0,0,0") -> (Pose) {
    angularDampingOverride=1
    angularSpringConstant=20000
    bAngularSpring=true
    bOverrideAngularDamping=true
    bUseGravityOverride=true
    gravityScale=0.35
    simulationSpace=RootRelative
    alphaBoolBlend {blendOption=Linear}
    boundBone {boneName=base_bubble_07}
    componentAppliedLinearAccClamp=<10000 10000 10000>
    componentLinearAccScale=<1 1 1>
    componentLinearVelScale=<1 1 1>
    gravityOverride {z=-980}
    physicsBodyDefinitions [{
      boundBone {boneName=base_bubble_07}
      boxExtents=<30 50 50>
      constraintSetup {angularLimitsMax=<2 3 4>, angularLimitsMin=<-3 -3 -3>, angularTarget=<1 0 0>}
      localJointOffset=<4 0 0>
     }]
   }
   n27 = Anim Dynamics - Bone: base_bubble_08!(SimulationSpace:AnimPhysSimSpaceType Enum, ComponentPose=@n26, Alpha=1, GravityOverride:Vector="0,0,-980", ExternalForce="0,0,0") -> (Pose) {
    angularDampingOverride=1
    angularSpringConstant=20000
    bAngularSpring=true
    bOverrideAngularDamping=true
    bUseGravityOverride=true
    gravityScale=0.35
    simulationSpace=RootRelative
    alphaBoolBlend {blendOption=Linear}
    boundBone {boneName=base_bubble_08}
    componentAppliedLinearAccClamp=<10000 10000 10000>
    componentLinearAccScale=<1 1 1>
    componentLinearVelScale=<1 1 1>
    gravityOverride {z=-980}
    physicsBodyDefinitions [{
      boundBone {boneName=base_bubble_08}
      boxExtents=<30 50 50>
      constraintSetup {angularLimitsMax=<2 3 4>, angularLimitsMin=<-3 -3 -3>, angularTarget=<1 0 0>}
      localJointOffset=<4 0 0>
     }]
   }
   Save cached pose 'CR_Cache_dynamics'(Pose=Component To Local(ComponentPose=@n27)) {cacheName=CR_Cache_dynamics}
  ]}
 EventGraph {nodes [
   Event BlueprintUpdateAnimation()
   > n1 = Sequence!(execute) -> (then_0, then_1, then_2:Exec)
   > n2 = Gravity Override State Function DEPRECATED(
    CharacterPartAnimInstance=Self-Reference()
    GameTimeInSeconds=GetGameTimeInSeconds()
    GravityOverrideParams=Get Gravity Override Params()
    RBAN Params=Get RBAN Params()
   )
   > Set GravityOverride(GravityOverride=@n2)
   > n3 = Branch(Condition=Get bIsInFrontEnd())
   > Set SimlationSpace(SimlationSpace:AnimPhysSimSpaceType Enum=World) -> (Output_Get:AnimPhysSimSpaceType Enum)
   Set SimlationSpace(execute=@n3.else)
   n4 = GetRealTimeSeconds()
   n5 = "Sin (Radians)"(A=float * float(A=float * float(A=0.9, B=1.1), B=@n4))
   n6 = "Sin (Radians)"(A=float * float(A=float * float(A=1.1, B=1.1), B=@n4))
   n7 = "Sin (Radians)"(A=float * float(A=@n4, B=float * float(A=1.2, B=5)))
   Set BubblesRotation1(
    execute=@n1.then_1
    BubblesRotation1:Rotator=MakeRotator(
     Roll=float * float(A=@n6, B=-2)
     Pitch=float * float(A=@n5, B=3)
     Yaw=Remap Float(Input=@n7, OldRange=(X=-1,Y=1), NewRange=(X=-0.2,Y=0.2)) -> (Output=~)
    )
   ) -> (Output_Get:Rotator)
   n8 = GetRealTimeSeconds()
   n9 = "Sin (Radians)"(A=float * float(A=float * float(A=0.9, B=1.1), B=@n8))
   n10 = "Sin (Radians)"(A=float * float(A=float * float(A=1.1, B=1.1), B=@n8))
   n11 = "Sin (Radians)"(A=float * float(A=@n8, B=float * float(A=1.2, B=5)))
   > Set BubblesRotation2(BubblesRotation2:Rotator=MakeRotator(
     Roll=float * float(A=@n10, B=2)
     Pitch=float * float(A=@n9, B=-1.5)
     Yaw=Remap Float(Input=@n11, OldRange=(X=-1,Y=1), NewRange=(X=-0.25,Y=0.25)) -> (Output=~)
    )) -> (Output_Get:Rotator)
   n12 = GetRealTimeSeconds()
   n13 = "Sin (Radians)"(A=float * float(A=float * float(A=0.9, B=1.1), B=@n12))
   n14 = "Sin (Radians)"(A=float * float(A=float * float(A=1.1, B=1.1), B=@n12))
   n15 = "Sin (Radians)"(A=float * float(A=@n12, B=float * float(A=1.2, B=5)))
   > Set BubblesRotation3(BubblesRotation3:Rotator=MakeRotator(
     Roll=float * float(A=@n14, B=2)
     Pitch=float * float(A=@n13, B=4.5)
     Yaw=Remap Float(Input=@n15, OldRange=(X=-1,Y=1), NewRange=(X=-0.33,Y=0.33)) -> (Output=~)
    )) -> (Output_Get:Rotator)
   n16 = GetRealTimeSeconds()
   n17 = "Sin (Radians)"(A=float * float(A=float * float(A=0.9, B=1.1), B=@n16))
   n18 = "Sin (Radians)"(A=float * float(A=float * float(A=1.1, B=1.1), B=@n16))
   n19 = "Sin (Radians)"(A=float * float(A=@n16, B=float * float(A=1.2, B=5)))
   > Set BubblesRotation4(BubblesRotation4:Rotator=MakeRotator(
     Roll=float * float(A=@n18, B=-1.2)
     Pitch=float * float(A=@n17, B=-0.5)
     Yaw=Remap Float(Input=@n19, OldRange=(X=-1,Y=1), NewRange=(X=-0.3,Y=0.3)) -> (Output=~)
    )) -> (Output_Get:Rotator)
   n20 = GetRealTimeSeconds()
   n21 = float * float(A=@n20, B=float * float(A=-0.68, B=0.68))
   n22 = Remap Float(
    Input="Atan (Radians)"!(A:double="Tan (Radians)"!(A:double=@n21) -> (ReturnValue:double)) -> (ReturnValue:double)
    OldRange=(X=-1,Y=1)
    NewRange=(X=0.9,Y=0.2)
   ) -> (Output=~)
   > Set BubblesScale1(BubblesScale1:Vector=MakeVector(X=@n22, Y=@n22, Z=@n22)) -> (Output_Get:Vector)
   n23 = float * float(A=@n20, B=float * float(A=-0.7, B=0.7))
   n24 = Remap Float(Input="Atan (Radians)"(A="Tan (Radians)"(A=@n23)), OldRange=(X=-1,Y=1), NewRange=(X=1.3,Y=0.2)) -> (Output=~)
   > Set BubblesScale2(BubblesScale2:Vector=MakeVector(X=@n24, Y=@n24, Z=@n24)) -> (Output_Get:Vector)
   n25 = float * float(A=@n20, B=float * float(A=-0.8, B=0.8))
   n26 = Remap Float(Input="Atan (Radians)"(A="Tan (Radians)"(A=@n25)), OldRange=(X=-1,Y=1), NewRange=(X=1.2,Y=0.2)) -> (Output=~)
   > Set BubblesScale3(BubblesScale3:Vector=MakeVector(X=@n26, Y=@n26, Z=@n26)) -> (Output_Get:Vector)
   n27 = float * float(A=@n20, B=float * float(A=-0.65, B=0.65))
   n28 = Remap Float(Input="Atan (Radians)"(A="Tan (Radians)"(A=@n27)), OldRange=(X=-1,Y=1), NewRange=(X=1.4,Y=0.2)) -> (Output=~)
   > Set BubblesScale4(BubblesScale4:Vector=MakeVector(X=@n28, Y=@n28, Z=@n28)) -> (Output_Get:Vector)
   n29 = float * float(A=@n20, B=float * float(A=-0.75, B=0.75))
   n30 = Remap Float(Input="Atan (Radians)"(A="Tan (Radians)"(A=@n29)), OldRange=(X=-1,Y=1), NewRange=(X=1.3,Y=0.2)) -> (Output=~)
   > Set BubblesScale5(BubblesScale5:Vector=MakeVector(X=@n30, Y=@n30, Z=@n30)) -> (Output_Get:Vector)
   n31 = float * float(A=@n20, B=float * float(A=-0.85, B=0.85))
   n32 = Remap Float(Input="Atan (Radians)"(A="Tan (Radians)"(A=@n31)), OldRange=(X=-1,Y=1), NewRange=(X=0.95,Y=0.2)) -> (Output=~)
   > Set BubblesScale6(BubblesScale6:Vector=MakeVector(X=@n32, Y=@n32, Z=@n32)) -> (Output_Get:Vector)
   n33 = float * float(A=@n20, B=float * float(A=-0.72, B=0.72))
   n34 = Remap Float(Input="Atan (Radians)"(A="Tan (Radians)"(A=@n33)), OldRange=(X=-1,Y=1), NewRange=(X=1.35,Y=0.2)) -> (Output=~)
   > Set BubblesScale7(BubblesScale7:Vector=MakeVector(X=@n34, Y=@n34, Z=@n34)) -> (Output_Get:Vector)
   n35 = float * float(A=@n20, B=float * float(A=-0.6, B=0.6))
   n36 = Remap Float(Input="Atan (Radians)"(A="Tan (Radians)"(A=@n35)), OldRange=(X=-1,Y=1), NewRange=(X=1.35,Y=0.2)) -> (Output=~)
   > n37 = Set BubblesScale8(BubblesScale8:Vector=MakeVector(X=@n36, Y=@n36, Z=@n36)) -> (Output_Get:Vector)
   n38 = Event BlueprintProcessAnimInputEvent()
   n39 = Branch(execute=@n37.then, Condition=OR Boolean(A=IsNamed(self=@n38.AnimInputEvent, InEventName=aie.Player.Zipline.Entered), B=Get bIsInFrontEnd()))
   > Set OnZipline(OnZipline:bool=true) -> (Output_Get:bool)
   > Set AlphaFE(AlphaFE:double=0.4) -> (Output_Get:double)
   Branch(
    execute=@n39.else
    Condition=OR Boolean!(A=IsNamed(self=@n38.AnimInputEvent, InEventName=aie.Player.Zipline.Exited), B=Get bIsMoving2D(), C:bool=Get bIsJumping()) -> (ReturnValue)
   )
   > Set OnZipline(OnZipline=false)
   > Set AlphaFE(AlphaFE=1)
   TryGetPawnOwner()
  ]}
]
variables [
 AlphaFE {type=number}
 BubblesRotation1 {type=Rotator}
 BubblesRotation2 {type=Rotator}
 BubblesRotation3 {type=Rotator}
 BubblesRotation4 {type=Rotator}
 BubblesScale1 {type=Vector}
 BubblesScale2 {type=Vector}
 BubblesScale3 {type=Vector}
 BubblesScale4 {type=Vector}
 BubblesScale5 {type=Vector}
 BubblesScale6 {type=Vector}
 BubblesScale7 {type=Vector}
 BubblesScale8 {type=Vector}
 Character Part Anim Instance {type=CustomCharacterPartAnimInstance, default=null}
 Gravity Override Params {type=GravityOverrideParamsStruct}
 GravityOverride {type=Vector}
 IsJumping {type=number}
 OnZipline {type=boolean}
 RBAN Params {
  type=RBANParamsStruct
  default {
   dBNO {simSpaceSettings {
     maxAngularAcceleration=1000
     maxAngularVelocity=3
     maxLinearAcceleration=300
     maxLinearVelocity=700
     velocityScaleZ=0
     externalLinearDragV=<0 0.5 0>
    }}
   dBNO_Carried {simSpaceSettings {
     maxAngularAcceleration=1000
     maxAngularVelocity=3
     maxLinearAcceleration=300
     maxLinearVelocity=700
     velocityScaleZ=0
     externalLinearDragV=<0 0.5 0>
    }}
   emoteOrMelee {simSpaceSettings {
     maxAngularAcceleration=1000
     maxAngularVelocity=3
     maxLinearAcceleration=300
     maxLinearVelocity=700
     velocityScaleZ=0
     externalLinearDragV=<0 0.5 0>
    }}
   front_End {simSpaceSettings {
     maxAngularAcceleration=1000
     maxAngularVelocity=3
     maxLinearAcceleration=300
     maxLinearVelocity=700
     velocityScaleZ=0
     externalLinearDragV=<0 0.5 0>
    }}
   onGround_Crouch_Moving {simSpaceSettings {
     maxAngularAcceleration=1000
     maxAngularVelocity=3
     maxLinearAcceleration=300
     maxLinearVelocity=700
     velocityScaleZ=0
     externalLinearDragV=<0 0.5 0>
    }}
   onGround_Falling {simSpaceSettings {
     maxAngularAcceleration=1000
     maxAngularVelocity=3
     maxLinearAcceleration=300
     maxLinearVelocity=700
     velocityScaleZ=0
     externalLinearDragV=<0 0.5 0>
    }}
   onGround_Moving {simSpaceSettings {
     maxAngularAcceleration=1000
     maxAngularVelocity=3
     maxLinearAcceleration=300
     maxLinearVelocity=700
     velocityScaleZ=0
     externalLinearDragV=<0 0.5 0>
    }}
   onGround_Standing {simSpaceSettings {
     maxAngularAcceleration=1000
     maxAngularVelocity=3
     maxLinearAcceleration=300
     maxLinearVelocity=700
     velocityScaleZ=0
     externalLinearDragV=<0 0.5 0>
    }}
   rail_Grinding {simSpaceSettings {
     maxAngularAcceleration=1000
     maxAngularVelocity=3
     maxLinearAcceleration=300
     maxLinearVelocity=700
     velocityScaleZ=0
     worldAlpha=0.1
     externalLinearDragV=<0 0.5 0>
    }}
   riding_Moving {simSpaceSettings {
     maxAngularAcceleration=1000
     maxAngularVelocity=3
     maxLinearAcceleration=300
     maxLinearVelocity=700
     velocityScaleZ=0
     externalLinearDragV=<0 0.5 0>
    }}
   riding_Moving_Targeting {simSpaceSettings {
     maxAngularAcceleration=1000
     maxAngularVelocity=3
     maxLinearAcceleration=300
     maxLinearVelocity=700
     velocityScaleZ=0
     externalLinearDragV=<0 0.5 0>
    }}
   riding_Standing {simSpaceSettings {
     maxAngularAcceleration=1000
     maxAngularVelocity=3
     maxLinearAcceleration=300
     maxLinearVelocity=700
     velocityScaleZ=0
     externalLinearDragV=<0 0.5 0>
    }}
   skydive_DivingDown {simSpaceSettings {
     maxAngularAcceleration=1000
     maxAngularVelocity=3
     maxLinearAcceleration=300
     maxLinearVelocity=700
     velocityScaleZ=0
     externalLinearDragV=<0 0.5 0>
    }}
   skydive_FloatingDown {
    componentLinearAccScale=<0 0.5 0>
    componentLinearVelScale=<0 0.5 0>
    simSpaceSettings {
     maxAngularAcceleration=1000
     maxAngularVelocity=3
     maxLinearAcceleration=300
     maxLinearVelocity=700
     velocityScaleZ=0
     externalLinearDragV=<0 0.5 0>
    }
   }
   skydive_GoingUp {simSpaceSettings {
     maxAngularAcceleration=1000
     maxAngularVelocity=3
     maxLinearAcceleration=300
     maxLinearVelocity=700
     velocityScaleZ=0
     externalLinearDragV=<0 0.5 0>
    }}
   skydive_Parachuting {
    componentLinearAccScale=<0 0.1 0.3>
    componentLinearVelScale=<0 0.1 0.3>
    simSpaceSettings {
     maxAngularAcceleration=1000
     maxAngularVelocity=3
     maxLinearAcceleration=300
     maxLinearVelocity=700
     velocityScaleZ=0
     externalLinearDragV=<0 0.5 0>
    }
   }
   skydive_Surfing {simSpaceSettings {
     maxAngularAcceleration=1000
     maxAngularVelocity=3
     maxLinearAcceleration=300
     maxLinearVelocity=700
     velocityScaleZ=0
     externalLinearDragV=<0 0.5 0>
    }}
   swimming {simSpaceSettings {
     maxAngularAcceleration=1000
     maxAngularVelocity=3
     maxLinearAcceleration=300
     maxLinearVelocity=700
     velocityScaleZ=0
     externalLinearDragV=<0 0.5 0>
    }}
   tactical_Sprint {simSpaceSettings {
     maxAngularAcceleration=1000
     maxAngularVelocity=3
     maxLinearAcceleration=300
     maxLinearVelocity=700
     velocityScaleZ=0
     externalLinearDragV=<0 0.5 0>
    }}
   ziplining {simSpaceSettings {
     maxAngularAcceleration=1000
     maxAngularVelocity=3
     maxLinearAcceleration=300
     maxLinearVelocity=700
     velocityScaleZ=0
     externalLinearDragV=<0 0.5 0>
    }}
  }
 }
 Sim Space Settings {
  type=SimSpaceSettings
  default {
   dampingAlpha=1
   maxAngularAcceleration=10000
   maxAngularVelocity=10000
   maxLinearAcceleration=10000
   maxLinearVelocity=10000
   velocityScaleZ=1
   worldAlpha=0
   externalLinearDragV=<0 0 0>
  }
 }
 SimlationSpace {type=AnimPhysSimSpaceType}
 Stream {type=RandomStream, default {initialSeed=20, seed=0}}
]
